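# Physics asset for the DashSurge male medium head cosmetic, providing capsule bodies and joints for dynamic hair and tongue bones.
# Four hair capsules (front, back, left, right) branch from a small hair base body constrained to the large head capsule.
# The tongue is a three-segment chain of tapering capsules on a small base body, whose joint to the head has free swing and twist.
# All other joints use angle limits, with tight twist throughout.
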PhysicsAsset /BRCosmetics/Characters/Player/Male/Medium/Heads/M_MED_DashSurge_Head/Meshes/M_MED_DashSurge_Head_Physics
bodies {
 dyn_hair_base [{shapeType=Capsule, length=0.505, radius=0.505, shapeName=dyn_hair_base_capsule, rotation=<rot 0 -90 0>}]
 dyn_hair_bk [{shapeType=Capsule, length=5, radius=1, center=<4 0 0>, rotation=<rot 90 0 0>}]
 dyn_hair_fr [{shapeType=Capsule, length=5, radius=1, center=<4 0 0>, rotation=<rot 90 0 0>}]
 dyn_hair_l [{shapeType=Capsule, length=5, radius=1, center=<4 0 0>, rotation=<rot 90 0 0>}]
 dyn_hair_r [{shapeType=Capsule, length=5, radius=1, center=<4 0 0>, rotation=<rot 90 0 0>}]
 dyn_tongue_01 [{shapeType=Capsule, length=4.5, radius=3.5, shapeName=dyn_tongue_01_capsule, center=<3 0 0>, rotation=<rot 0 0 90>}]
 dyn_tongue_02 [{shapeType=Capsule, length=6, radius=3, shapeName=dyn_tongue_02_capsule, center=<3 0 0>, rotation=<rot 0 0 90>}]
 dyn_tongue_03 [{shapeType=Capsule, length=6, radius=2, shapeName=dyn_tongue_03_capsule, center=<3 0 0>, rotation=<rot 0 0 90>}]
 dyn_tongue_base [{shapeType=Capsule, length=0.505, radius=0.505, shapeName=dyn_tongue_base_capsule, rotation=<rot 0 -90 0>}]
 head [{shapeType=Capsule, length=20, radius=15, center=<-5.0518 -1.3597 0>}]
}
constraints [
 {bone1Name=dyn_tongue_03, bone2Name=dyn_tongue_02, swing1LimitDegrees=10, swing2LimitDegrees=40, twistLimitDegrees=5}
 {
  bone1Name=dyn_tongue_base
  bone2Name=head
  swing1LimitDegrees=45
  swing1Motion=Free
  swing2LimitDegrees=45
  swing2Motion=Free
  twistLimitDegrees=45
  twistMotion=Free
 }
 {bone1Name=dyn_tongue_01, bone2Name=dyn_tongue_base, swing1LimitDegrees=10, swing2LimitDegrees=50, twistLimitDegrees=5}
 {bone1Name=dyn_tongue_02, bone2Name=dyn_tongue_01, swing1LimitDegrees=10, swing2LimitDegrees=40, twistLimitDegrees=5}
 {bone1Name=dyn_hair_r, bone2Name=dyn_hair_base, swing1LimitDegrees=5, swing2LimitDegrees=30, twistLimitDegrees=5}
 {bone1Name=dyn_hair_base, bone2Name=head, swing1LimitDegrees=5, swing2LimitDegrees=30, twistLimitDegrees=5}
 {bone1Name=dyn_hair_bk, bone2Name=dyn_hair_base, swing1LimitDegrees=5, swing2LimitDegrees=30, twistLimitDegrees=5}
 {bone1Name=dyn_hair_fr, bone2Name=dyn_hair_base, swing1LimitDegrees=5, swing2LimitDegrees=30, twistLimitDegrees=5}
 {bone1Name=dyn_hair_l, bone2Name=dyn_hair_base, swing1LimitDegrees=5, swing2LimitDegrees=30, twistLimitDegrees=5}
]
